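# Unlit, translucent, two-sided material for Niagara mesh particles in the Rockstar Sparkle skin's ring effect.
# Scalar-driven lerps blend between ring and sprite looks within one material.
# Color mixes a curve-atlas gradient or sprite tint with particle color; opacity selects among view-angle fade, flipbook mask and spherical glow mask.
# World position offset thickens the mesh via SplineThicken, or billboards it with Sprite plus a local offset.
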Material /Game/Characters/Player/Female/Medium/Bodies/F_MED_Rockstar/Skins/Sparkle/FX/MI_Rockstar_Sparkle_Ring
graph {
 properties {
  bUsedWithNiagaraMeshParticles=true
  blendMode=BLEND_Translucent
  pixelDepthOffsetMode=PDOM_Legacy
  shadingModel=MSM_Unlit
  twoSided=true
  usedShadingModels=Unlit
 }
 nodes [
  n1 = ParticleColor()
  n2 = ScalarParameter() {defaultValue=0.8, parameterName=MeshFade}
  n3 = Constant() {r=2}
  n4 = MaterialFunctionCall(
   "Animation  Phase (0-1)"=Constant() {r=0.5}
   Number of Rows=@n3
   Number of Columns=@n3
   Texture=TextureObject() {texture=/Game/Accessories/FORT_Backpacks/M_MED_Sleepy_Time_Backpack/FX/T_SleepyTime_Elements}
  ) {materialFunction=&FlipBook}
  n5 = Multiply(
   A=StaticComponentMaskParameter(None=Multiply(A=@n4.Result)) {defaultR=true, parameterName=Mask}
   B=ScalarParameter() {defaultValue=1, parameterName=Sprite Alpha}
  )
  n6 = ScalarParameter() {parameterName=Switch_Sprite}
  n7 = Power(
   Base=SphereMask(A=TextureCoordinate() {vTiling=0}, B=Constant2Vector() {r=0.5}) {attenuationRadius=0.5, hardnessPercent=0}
   Exp=ScalarParameter() {defaultValue=3, parameterName=Ring edge mask power}
  ) {constExponent=3}
  n8 = ParticleColor()
  LinearInterpolate(
   A=Multiply(
    A=@n1.A
    B=MaterialFunctionCall("FadeSensitivity (optional)"=@n2) {materialFunction=&FadeBasedOnViewAngle}.Result
   )
   B=LinearInterpolate(A=LinearInterpolate(B=@n5, Alpha=@n6) {constA=1}, B=Multiply(A=@n7, B=@n8.A), Alpha=ScalarParameter() {parameterName=Use ring glow})
   Alpha=ScalarParameter() {defaultValue=1, parameterName=Use MeshMask}
  )
  n9 = CurveAtlasRowParameter(CurveTime=ComponentMask(None=TextureCoordinate()) {g=true}) {
   atlas=&/CA_RockStar_Sparkle
   curve=&/CV_RockStar_Sparkle
   parameterName=CV_Color
  }
  n10 = LinearInterpolate(A=@n9, Alpha=ScalarParameter() {parameterName=Use ring glow})
  n11 = VectorParameter() {parameterName=Sprite Color, defaultValue=<rgba 0.3144 0.2917 1 1>}
  MaterialFunctionCall(Emissive=LinearInterpolate(A=Multiply(A=@n1.RGB, B=@n10), B=Multiply(A=@n11.RGB, B=@n1.RGB), Alpha=@n6) {constA=1}) {materialFunction=&MF_FX_EmissiveAdjustments}
  n12 = ScalarParameter() {defaultValue=0.7, parameterName=Ring Width}
  n13 = MaterialFunctionCall(WidthBase=@n12, WidthTip=@n12) {materialFunction=&SplineThicken}
  n14 = Add(
   A=MaterialFunctionCall!(Center Location, 0-1 UVs, XY scale=@n12, OverridePivot, "Normals (Optional)") {materialFunction=&Sprite}.World position offset
   B=Transform(None=VectorParameter() {parameterName=Sprite Offset, defaultValue=<rgba 0 0 12 1>}.RGB) {transformSourceType=TRANSFORMSOURCE_Local}
  )
  Reroute(None=LinearInterpolate(
    A=Constant()
    B=LinearInterpolate(A=@n13.WorldPositionOffset, B=@n14, Alpha=@n6) {constB=0}
    Alpha=ScalarParameter() {defaultValue=1, parameterName=Use SplineThicken}
   ))
 ]
}
